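; Parent Blueprint for Primal runestone held objects that give the carrier tracking and low-gravity effects.
; On attach it applies both GameplayEffects and an ambient cue, with a timer reapplying low gravity if lost outside swimming.
; A second timer feeds held state and the authority-set obelisk position to the idle Niagara FX.
; Drop, throw, landing, place and destroy events clear the effects and timers, then either deactivate FX or remove the cue.
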(event Collision|EventActorBeginOverlap
  (params (OtherActor :type Actor%20Object%20Reference)))

(event Custom|UpdateParticleFXOnPickup
  ;[1 1 1] On Begin - Turn on our idle FX
  (Niagara|SetNiagaraVariable(Bool) (Variables|Default|GetNS_Gadget_TowerPickup_Idle) "bIsHeld" false)
  ;[1 1 1] Periodically send updated closest obelisk position to FX
  (bind _returnvalue (Utilities|Time|SetTimerbyEvent (delegate Custom|SendObeliskPos) 1.0 true))
  (Variables|Default|SetSendObeliskPosTimer _returnvalue))

(event EventBeginPlay
  ;[1 1 1] On Begin - Turn on our idle FX
  (|CallParentFunction)
  (Niagara|SetNiagaraVariable(Bool) (Variables|Default|GetNS_Gadget_TowerPickup_Idle) "bIsHeld" false)
  ;[1 1 1] Periodically send updated closest obelisk position to FX
  (bind _returnvalue (Utilities|Time|SetTimerbyEvent (delegate Custom|SendObeliskPos) 1.0 true))
  (Variables|Default|SetSendObeliskPosTimer _returnvalue))

(event Custom|SendObeliskPos
  (bind _isheld (Variables|Default|GetIsHeld))
  (bind _ns_gadget_towerpickup_idle (Variables|Default|GetNS_Gadget_TowerPickup_Idle))
  ;[1 1 1] Periodically send updated closest obelisk position to FX
  (Niagara|SetNiagaraVariable(Vector3) _ns_gadget_towerpickup_idle "ObeliskPosition" (Variables|Default|GetObeliskPosition))
  (Effects|Components|ParticleSystem|SetBoolParameter _ns_gadget_towerpickup_idle "bIsFacing" _isheld)
  (Effects|Components|ParticleSystem|SetBoolParameter _ns_gadget_towerpickup_idle "bIsHeld" _isheld))

(event OnHeldObjectPickedUp(FortHeldObject)
  ;[1 1 1] On Picked up - Check we were picked up by a pawn (Why wouldn't we be?) and set Is Picked Up to True
  (bind _returnvalue (Fortnite|HeldObjects|GetHeldObjectsOwningPlayer self))
  (Utilities|Casting|CastToFortPlayerPawnAthena _returnvalue)
  (CallFunction|UpdateParticleFXOnPickup))

(event Custom|ReApplyLowGravGE
  ;[1 1 1] Grant the picking up player the Tracking Gameplay Effect and the LowGravity Gameplay Effect - Then set up a timer to keep checking that the pawn has the lowgrav effect as there are areas or scenarios when we remove it.
  (bind _returnvalue (GameplayEffects|FortApplyGameplayEffectToSelf :self (Ability|GetFortAbilitySystemComponent (Game|Player|GetPlayerPawn (stale OnHeldObjectAttached(FortHeldObject) _output_get_2))) :GameplayEffectClass "/PrimalGameplay/Items/HeldObjects/GE_Primal_Runestone_LowGrav.GE_Primal_Runestone_LowGrav_C" :Level 0.0))
  (Variables|Default|SetLowGravGE _returnvalue)
  (bind _returnvalue_2 (Utilities|Time|SetTimerbyEvent (delegate Custom|CheckToReApplyLowGrav) 3.0 true))
  (Variables|Default|SetLowGravCheckTimer _returnvalue_2)
  (Ability|AddGameplayCueReplicated (Variables|Default|GetGCNL_AmbientLooping)))

(event OnHeldObjectAttached(FortHeldObject)
  (bind _self self)
  ;[1 1 1] On Attached - Check we are indeed being held, force a net update and store the player controller of the pawn who picked up the object
  (bind _returnvalue (Fortnite|HeldObjects|IsHeldObjectBeingHeld _self))
  (bind _output_get (Variables|Default|SetIsHeld _returnvalue))
  (if _output_get
    (Networking|ForceNetUpdate)
    (bind _returnvalue_2 (Fortnite|HeldObjects|GetHeldObjectsOwningPlayer _self))
    (bind _output_get_2 (Variables|Default|SetPlayerController (Fortnite|GetFortPlayerControllerfromActor _returnvalue_2)))
    ;[1 1 1] Grant the picking up player the Tracking Gameplay Effect and the LowGravity Gameplay Effect - Then set up a timer to keep checking that the pawn has the lowgrav effect as there are areas or scenarios when we remove it.
    (Utilities|IsValid (Game|Player|GetPlayerPawn _output_get_2)
      (:Is%20Valid
        (bind _returnvalue_3 (GameplayEffects|FortApplyGameplayEffectToSelf :self (Ability|GetFortAbilitySystemComponent (Game|Player|GetPlayerPawn _output_get_2)) :GameplayEffectClass "/PrimalGameplay/Items/HeldObjects/GE_Primal_TrackingActivate.GE_Primal_TrackingActivate_C" :Level 0.0))
        (Variables|Default|SetTrackingGE _returnvalue_3)
        (bind _returnvalue_4 (GameplayEffects|FortApplyGameplayEffectToSelf :self (Ability|GetFortAbilitySystemComponent (Game|Player|GetPlayerPawn _output_get_2)) :GameplayEffectClass "/PrimalGameplay/Items/HeldObjects/GE_Primal_Runestone_LowGrav.GE_Primal_Runestone_LowGrav_C" :Level 0.0))
        (Variables|Default|SetLowGravGE _returnvalue_4)
        (bind _returnvalue_5 (Utilities|Time|SetTimerbyEvent (delegate Custom|CheckToReApplyLowGrav) 3.0 true))
        (Variables|Default|SetLowGravCheckTimer _returnvalue_5)
        (Ability|AddGameplayCueReplicated (Variables|Default|GetGCNL_AmbientLooping))))))

(event Custom|CheckToReApplyLowGrav
  (bind _playercontroller (Variables|Default|GetPlayerController))
  (bind _returnvalue (Game|Player|GetPlayerPawn _playercontroller))
  ;[1 1 1] This timer checks repeatedly to ensure the player has the correct lowgrav effect and reapplies it if it falls off
  (if (AI|Components|NavMovement|IsSwimming (Variables|/Script/Engine.Character|GetCharacterMovement _returnvalue) :!class /Script/Engine.NavMovementComponent)
    _
    (elif (GameplayTags|HasMatchingGameplayTag (Ability|GetFortAbilitySystemComponent _returnvalue) "(TagName=\"Status.LowGravity.LowGravRock\")")
      _
      (else
        (CallFunction|ReApplyLowGravGE)))))

(event OnHeldObjectDropped(FortHeldObject)
  (bind _playercontroller (Variables|Default|GetPlayerController))
  (bind _returnvalue (Game|Player|GetPlayerPawn _playercontroller))
  (bind _returnvalue_2 (Ability|GetFortAbilitySystemComponent _returnvalue))
  ;[1 1 1] On dropped / removed etc handle clearing up and removing all relevant effects
  (Utilities|IsValid _playercontroller
    (:Is%20Valid
      (Utilities|IsValid _returnvalue
        (:Is%20Valid
          (Variables|Default|SetIsHeld false)
          (GameplayEffects|RemoveActiveGameplayEffect _returnvalue_2 (Variables|Default|GetTrackingGE) -1)
          (GameplayEffects|RemoveActiveGameplayEffect _returnvalue_2 (Variables|Default|GetLowGravGE) -1)
          (Utilities|Time|ClearandInvalidateTimerbyHandle (Variables|Default|GetSendObeliskPosTimer))
          (Utilities|Time|ClearandInvalidateTimerbyHandle (Variables|Default|GetLowGravCheckTimer))
          (Networking|CanExecuteCosmeticEvents
            (:True
              (Components|Activation|Deactivate (Variables|Default|GetNS_Gadget_TowerPickup_Idle)))
            (:False
              (Ability|RemoveGameplayCueReplicated (Variables|Default|GetGCNL_AmbientLooping)))))))))

(event OnHeldObjectDestroy(FortHeldObject)
  (bind _playercontroller (Variables|Default|GetPlayerController))
  (bind _returnvalue (Game|Player|GetPlayerPawn _playercontroller))
  (bind _returnvalue_2 (Ability|GetFortAbilitySystemComponent _returnvalue))
  ;[1 1 1] On dropped / removed etc handle clearing up and removing all relevant effects
  (Utilities|IsValid _playercontroller
    (:Is%20Valid
      (Utilities|IsValid _returnvalue
        (:Is%20Valid
          (Variables|Default|SetIsHeld false)
          (GameplayEffects|RemoveActiveGameplayEffect _returnvalue_2 (Variables|Default|GetTrackingGE) -1)
          (GameplayEffects|RemoveActiveGameplayEffect _returnvalue_2 (Variables|Default|GetLowGravGE) -1)
          (Utilities|Time|ClearandInvalidateTimerbyHandle (Variables|Default|GetSendObeliskPosTimer))
          (Utilities|Time|ClearandInvalidateTimerbyHandle (Variables|Default|GetLowGravCheckTimer))
          (Networking|CanExecuteCosmeticEvents
            (:True
              (Components|Activation|Deactivate (Variables|Default|GetNS_Gadget_TowerPickup_Idle)))
            (:False
              (Ability|RemoveGameplayCueReplicated (Variables|Default|GetGCNL_AmbientLooping)))))))))

(event OnHeldObjectPlaced(FortHeldObject)
  (bind _playercontroller (Variables|Default|GetPlayerController))
  (bind _returnvalue (Game|Player|GetPlayerPawn _playercontroller))
  (bind _returnvalue_2 (Ability|GetFortAbilitySystemComponent _returnvalue))
  ;[1 1 1] On dropped / removed etc handle clearing up and removing all relevant effects
  (Utilities|IsValid _playercontroller
    (:Is%20Valid
      (Utilities|IsValid _returnvalue
        (:Is%20Valid
          (Variables|Default|SetIsHeld false)
          (GameplayEffects|RemoveActiveGameplayEffect _returnvalue_2 (Variables|Default|GetTrackingGE) -1)
          (GameplayEffects|RemoveActiveGameplayEffect _returnvalue_2 (Variables|Default|GetLowGravGE) -1)
          (Utilities|Time|ClearandInvalidateTimerbyHandle (Variables|Default|GetSendObeliskPosTimer))
          (Utilities|Time|ClearandInvalidateTimerbyHandle (Variables|Default|GetLowGravCheckTimer))
          (Networking|CanExecuteCosmeticEvents
            (:True
              (Components|Activation|Deactivate (Variables|Default|GetNS_Gadget_TowerPickup_Idle)))
            (:False
              (Ability|RemoveGameplayCueReplicated (Variables|Default|GetGCNL_AmbientLooping)))))))))

(event OnHeldObjectThrownLanded(FortHeldObject)
  (bind _playercontroller (Variables|Default|GetPlayerController))
  (bind _returnvalue (Game|Player|GetPlayerPawn _playercontroller))
  (bind _returnvalue_2 (Ability|GetFortAbilitySystemComponent _returnvalue))
  ;[1 1 1] On dropped / removed etc handle clearing up and removing all relevant effects
  (Utilities|IsValid _playercontroller
    (:Is%20Valid
      (Utilities|IsValid _returnvalue
        (:Is%20Valid
          (Variables|Default|SetIsHeld false)
          (GameplayEffects|RemoveActiveGameplayEffect _returnvalue_2 (Variables|Default|GetTrackingGE) -1)
          (GameplayEffects|RemoveActiveGameplayEffect _returnvalue_2 (Variables|Default|GetLowGravGE) -1)
          (Utilities|Time|ClearandInvalidateTimerbyHandle (Variables|Default|GetSendObeliskPosTimer))
          (Utilities|Time|ClearandInvalidateTimerbyHandle (Variables|Default|GetLowGravCheckTimer))
          (Networking|CanExecuteCosmeticEvents
            (:True
              (Components|Activation|Deactivate (Variables|Default|GetNS_Gadget_TowerPickup_Idle)))
            (:False
              (Ability|RemoveGameplayCueReplicated (Variables|Default|GetGCNL_AmbientLooping)))))))))

(event OnHeldObjectThrown(FortHeldObject)
  (bind _playercontroller (Variables|Default|GetPlayerController))
  (bind _returnvalue (Game|Player|GetPlayerPawn _playercontroller))
  (bind _returnvalue_2 (Ability|GetFortAbilitySystemComponent _returnvalue))
  ;[1 1 1] On dropped / removed etc handle clearing up and removing all relevant effects
  (Utilities|IsValid _playercontroller
    (:Is%20Valid
      (Utilities|IsValid _returnvalue
        (:Is%20Valid
          (Variables|Default|SetIsHeld false)
          (GameplayEffects|RemoveActiveGameplayEffect _returnvalue_2 (Variables|Default|GetTrackingGE) -1)
          (GameplayEffects|RemoveActiveGameplayEffect _returnvalue_2 (Variables|Default|GetLowGravGE) -1)
          (Utilities|Time|ClearandInvalidateTimerbyHandle (Variables|Default|GetSendObeliskPosTimer))
          (Utilities|Time|ClearandInvalidateTimerbyHandle (Variables|Default|GetLowGravCheckTimer))
          (Networking|CanExecuteCosmeticEvents
            (:True
              (Components|Activation|Deactivate (Variables|Default|GetNS_Gadget_TowerPickup_Idle)))
            (:False
              (Ability|RemoveGameplayCueReplicated (Variables|Default|GetGCNL_AmbientLooping)))))))))

(fn ServerSetObeliskPosition
  (params (ObeliskPosition :type Vector))
  (Networking|SwitchHasAuthority
    (:Authority
      (Variables|Default|SetObeliskPosition ObeliskPosition))))

(fn ConstructionScript
  (|CallParentFunction))
