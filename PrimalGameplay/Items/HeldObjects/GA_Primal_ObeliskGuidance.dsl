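; Gameplay ability that guides a player carrying a Primal runestone toward the nearest obelisk with an attached Niagara trail.
; On activation it polls every second until the held runestone is found, then splits by authority.
; The authority periodically picks the closest obelisk and sets its position on the runestone for replication.
; Non-authority instances spawn the trail and feed it obelisk and runestone locations on a 60 Hz timer.
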(event Custom|Update%20Rotation
  (bind _guidanceeffect (Variables|Default|GetGuidanceEffect))
  ;[1 1 1] Update Rotation of FX based on Obelisk Location rep'd down from the Server
  (Effects|Components|ParticleSystem|SetVectorParameter _guidanceeffect "ObeliskLocation" (Variables|/PrimalGameplay/Items/HeldObjects/B_HeldObject_Primal_Runestone_Parent.B_HeldObject_Primal_Runestone_Parent_C|GetObeliskPosition (Variables|Default|GetRunestone)))
  (Effects|Components|ParticleSystem|SetVectorParameter _guidanceeffect "PlayerLocation" (Transformation|GetActorLocation (Variables|Default|GetRunestone))))

(event Ability|EventActivateAbility
  ;[1 1 1] Get the player on the quest, and if we fail, keep checking until we succeed
  (bind _asfort_player_controller_athena (Utilities|Casting|CastToFortPlayerControllerAthena (Fortnite|GetFortPlayerControllerfromActor (CallFunction|GetAvatarasFortPlayerPawn))))
  (Variables|Default|SetPlayerController _asfort_player_controller_athena)
  (bind _asfort_player_pawn (Utilities|Casting|CastToFortPlayerPawn (Ability|GetAvatarActorfromActorInfo)))
  (Variables|Default|SetPlayerPawn _asfort_player_pawn)
  (bind _playerpawn (Variables|Default|GetPlayerPawn)
    (:then
      (bind _returnvalue (Fortnite|HeldObjects|GetHeldObjectAttachedtoPlayer _playerpawn))
      (Utilities|IsValid _returnvalue
        (:Is%20Valid
          (Utilities|Time|ClearandInvalidateTimerbyHandle (Variables|Default|GetValidPlayerTimer))
          (bind _asb_held_object_primal_runestone_parent (Utilities|Casting|CastToB_HeldObject_Primal_Runestone_Parent _returnvalue))
          (Variables|Default|SetRunestone _asb_held_object_primal_runestone_parent)
          (Ability|HasAuthority
            (:True
              ;[1 1 1] Get closest obelisk and track to it
              (bind _returnvalue_2 (Utilities|Time|SetTimerbyEvent (delegate Custom|CheckClosestObelisk) 5.0 true false 1e+01))
              (Variables|Default|SetClosestObeliskTimer _returnvalue_2)
              (bind _outactors (Actor|GetAllActorsOfClass "/PrimalGameplay/Items/HeldObjects/Obelisk/BGA_Primal_Obelisk_Parent.BGA_Primal_Obelisk_Parent_C"))
              (bind _output_get (Variables|Default|SetObeliskActorArray _outactors))
              (bind _output_get_2 (Variables|Default|SetClosestObelisk (Fortnite|GetClosestActorfromArray (Variables|Default|GetPlayerPawn) _output_get)))
              (Class|BHeldObjectPrimalRunestoneParent|ServerSetObeliskPosition (Variables|Default|GetRunestone) (Transformation|GetActorLocation _output_get_2)))
            (:False
              ;[1 1 1] Update Rotation of FX based on Obelisk Location rep'd down from the Server
              (CallFunction|SpawnBreadcrumbTrail)
              (Utilities|Time|SetTimerbyEvent (delegate Custom|Update%20Rotation) 0.016667 true))))
        (:Is%20Not%20Valid
          (bind _returnvalue_3 (Utilities|Time|SetTimerbyEvent (delegate Custom|Recheck%20Player) 1.0 true))
          (Variables|Default|SetValidPlayerTimer _returnvalue_3))))
    (:else)))

(event Ability|EventFailedToActivatePassiveAbility
  (params (ActorInfo :type Gameplay%20Ability%20Actor%20Info%20Structure)))

(event Custom|Recheck%20Player
  ;[1 1 1] Get the player on the quest, and if we fail, keep checking until we succeed
  (bind _playerpawn (Variables|Default|GetPlayerPawn)
    (:then
      (bind _returnvalue (Fortnite|HeldObjects|GetHeldObjectAttachedtoPlayer _playerpawn))
      (Utilities|IsValid _returnvalue
        (:Is%20Valid
          (Utilities|Time|ClearandInvalidateTimerbyHandle (Variables|Default|GetValidPlayerTimer))
          (bind _asb_held_object_primal_runestone_parent (Utilities|Casting|CastToB_HeldObject_Primal_Runestone_Parent _returnvalue))
          (Variables|Default|SetRunestone _asb_held_object_primal_runestone_parent)
          (Ability|HasAuthority
            (:True
              ;[1 1 1] Get closest obelisk and track to it
              (bind _returnvalue_2 (Utilities|Time|SetTimerbyEvent (delegate Custom|CheckClosestObelisk) 5.0 true false 1e+01))
              (Variables|Default|SetClosestObeliskTimer _returnvalue_2)
              (bind _outactors (Actor|GetAllActorsOfClass "/PrimalGameplay/Items/HeldObjects/Obelisk/BGA_Primal_Obelisk_Parent.BGA_Primal_Obelisk_Parent_C"))
              (bind _output_get (Variables|Default|SetObeliskActorArray _outactors))
              (bind _output_get_2 (Variables|Default|SetClosestObelisk (Fortnite|GetClosestActorfromArray (Variables|Default|GetPlayerPawn) _output_get)))
              (Class|BHeldObjectPrimalRunestoneParent|ServerSetObeliskPosition (Variables|Default|GetRunestone) (Transformation|GetActorLocation _output_get_2)))
            (:False
              ;[1 1 1] Update Rotation of FX based on Obelisk Location rep'd down from the Server
              (CallFunction|SpawnBreadcrumbTrail)
              (Utilities|Time|SetTimerbyEvent (delegate Custom|Update%20Rotation) 0.016667 true))))
        (:Is%20Not%20Valid
          (bind _returnvalue_3 (Utilities|Time|SetTimerbyEvent (delegate Custom|Recheck%20Player) 1.0 true))
          (Variables|Default|SetValidPlayerTimer _returnvalue_3))))
    (:else)))

(event Custom|CheckClosestObelisk
  ;[1 1 1] Get closest obelisk and track to it
  (bind _output_get (Variables|Default|SetClosestObelisk (Fortnite|GetClosestActorfromArray (Variables|Default|GetPlayerPawn) (Variables|Default|GetObeliskActorArray))))
  (Class|BHeldObjectPrimalRunestoneParent|ServerSetObeliskPosition (Variables|Default|GetRunestone) (+ (Transformation|GetActorLocation _output_get) (Math|Vector|RotateVector (Variables|Default|GetObeliskOffset) (Transformation|GetActorRotation _output_get)))))

(event Custom|SpawnBreadcrumbTrail
  ;[1 1 1] Spawn the trail
  (bind _playercontroller (Variables|Default|GetPlayerController)
    (:then
      _)
    (:else
      (bind _asfort_player_controller_athena (Utilities|Casting|CastToFortPlayerControllerAthena (Fortnite|GetFortPlayerControllerfromActor (CallFunction|GetAvatarasFortPlayerPawn))))
      (Variables|Default|SetPlayerController _asfort_player_controller_athena)))
  (bind _returnvalue (Effects|SpawnFXSystemAttached "/PrimalGameplay/Items/HeldObjects/Assets/Tower_Obelisk/Effects/NS_TowerObelisk_Guidance.NS_TowerObelisk_Guidance" (Variables|/Script/Engine.Character|GetMesh (Game|Player|GetPlayerPawn _playercontroller))))
  (Variables|Default|SetGuidanceEffect _returnvalue))

(event Ability|EventOnEndAbility
  (params (bWasCancelled :type Boolean))
  (bind _guidanceeffect (Variables|Default|GetGuidanceEffect))
  ;[1 1 1] On End clean up
  (Components|Activation|Deactivate _guidanceeffect)
  (Components|DestroyComponent _guidanceeffect)
  (Utilities|Time|ClearandInvalidateTimerbyHandle (Variables|Default|GetClosestObeliskTimer)))

(fn RotationToObelisk
  (return (Math|Rotator|MakeRotfromX (* (- (Transformation|GetActorLocation (Game|Player|GetPlayerPawn (Variables|Default|GetPlayerController))) (Transformation|GetActorLocation (Variables|Default|GetClosestObelisk))) -1.0))))

(fn SortArrayByDistanceToPlayer
  (params (Array :type Array%20of%20Actor%20Object%20References) (Pawn :type Fort%20Player%20Pawn%20Object%20Reference))
  (for (_array_element _array_index) Array
    (Variables|Default|SetCurrentTarget _array_element)
    (Variables|Default|SetCurrentTargetStepper _array_index)
    (Variables|Default|SetCurrentTargetDistance (Class|AbilityFunctions|GetDistanceBetweenActors (Variables|Default|GetCurrentTarget) Pawn))
    (while (and (> (Class|AbilityFunctions|GetDistanceBetweenActors Pawn (Utilities|Array|Get(acopy) (Variables|Default|GetSortedList) (- (Variables|Default|GetCurrentTargetStepper) 1))) (Variables|Default|GetCurrentTargetDistance)) (> (Variables|Default|GetCurrentTargetStepper) 0))
      (Variables|Default|SetCurrentTargetStepper (- (Variables|Default|GetCurrentTargetStepper) 1)))
    (Utilities|Array|Insert (Variables|Default|GetSortedList) (Variables|Default|GetCurrentTarget) (Variables|Default|GetCurrentTargetStepper)))
  (return (Variables|Default|GetSortedList)))
